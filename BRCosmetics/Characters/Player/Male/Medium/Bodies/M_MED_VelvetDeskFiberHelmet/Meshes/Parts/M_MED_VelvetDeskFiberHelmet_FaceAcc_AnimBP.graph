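# Animation Blueprint for the face accessory mesh of the VelvetDeskFiberHelmet male medium player cosmetic.
# The AnimGraph copies the pose from the attached parent mesh, then runs it through the accessory's Control Rig to produce the output pose.
# When character customization loads, it stores the Head part's mesh component in HeadSkelMesh, which nothing shown here reads.
# The per-frame update event is empty.
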AnimBlueprint /BRCosmetics/Characters/Player/Male/Medium/Bodies/M_MED_VelvetDeskFiberHelmet/Meshes/Parts/M_MED_VelvetDeskFiberHelmet_FaceAcc_AnimBP
target_skeleton=&/M_MED_VelvetDeskFiberHelmet_FaceAcc_Skeleton
graphs [
 AnimGraph {nodes [
   n1 = Control Rig(Source=Copy Pose From Mesh() {bUseAttachedParent=true}) {
    alphaBoolBlend {blendOption=Linear}
    controlRigAssetReference {blueprintRigClass=&/M_MED_velvetDeskFiber_FaceAcc_CtrlRig.M_MED_velvetDeskFiber_FaceAcc_CtrlRig_C}
   }
   Output Pose(Result=@n1) {layerGroup=None}
  ]}
 EventGraph {nodes [
   n1 = Event OnCharacterCustomizationLoaded()
   > Set HeadSkelMesh(HeadSkelMesh:Skeletal Mesh Component Budgeted Object Reference=Get Array Item with EFort Custom Part Type(Array=@n1.CustomMeshParts, PartType=Head)) -> (Output_Get:Skeletal Mesh Component Budgeted Object Reference)
   Event BlueprintUpdateAnimation()
  ]}
]
variables [HeadSkelMesh {type=SkeletalMeshComponentBudgeted}]
